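; Parent Blueprint for table lamp light-source props, built on a Fortnite building actor class.
; Its construction script runs the parent's, then reads the first building texture data entry, if one exists.
; A valid diffuse texture from that entry is pushed into the "Diffuse" parameter of the static mesh's dynamic material instance, as a workaround for custom materials on the parent class.
; The BeginPlay, overlap and Tick events are empty.
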(event EventBeginPlay)

(event Collision|EventActorBeginOverlap
  (params (OtherActor :type Actor%20Object%20Reference)))

(event EventTick
  (params (DeltaSeconds :type Float%20%28single-precision%29)))

(fn ConstructionScript
  (|CallParentFunction)
  ;[1 1 1] Workaround for custom materials being used on parent class
  (bind _outtexturedataarray (Building|GetBuildingTextureData))
  (if (> (Utilities|Array|Length _outtexturedataarray) 0)
    (if (Utilities|IsValid (Variables|/Script/FortniteGame.BuildingTextureData|GetDiffuse (Utilities|Array|Get(acopy) _outtexturedataarray)) :!class /Script/Engine.KismetSystemLibrary)
      (Rendering|Material|SetTextureParameterValue (Class|BTechArtGlobalLibrary|GetDynamicMaterialInstance (Variables|Debug|GetStaticMeshComponent)) "Diffuse" (Variables|/Script/FortniteGame.BuildingTextureData|GetDiffuse (Utilities|Array|Get(acopy) _outtexturedataarray))))))
